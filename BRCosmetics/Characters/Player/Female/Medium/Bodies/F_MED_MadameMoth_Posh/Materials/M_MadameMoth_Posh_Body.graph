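# Body material for the MadameMoth Posh cosmetic character, adding animated flame-like fresnel glow to masked areas.
# It wraps the effect between the shared CharacterParent 2023 PreFX and PostFX functions.
# Scrolling fire noise modulates the fresnel, driving tinted emissive and darkening base color within a channel-selectable mask.
# A second opposite-scrolling noise layer is averaged in, except on low and medium quality or deferred and mobile shading paths.
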Material /BRCosmetics/Characters/Player/Female/Medium/Bodies/F_MED_MadameMoth_Posh/Materials/M_MadameMoth_Posh_Body
graph {
 properties {
  bUseMaterialAttributes=true
  bUsedWithMeshDeformer=true
  bUsedWithMorphTargets=true
  bUsedWithSkeletalMesh=true
  pixelDepthOffsetMode=PDOM_Legacy
  shadingModel=MSM_FromMaterialExpression
  usedShadingModels=Default Lit | Subsurface | Clear Coat | Subsurface Profile | Cloth
 }
 nodes [
  n1 = MaterialFunctionCall() {materialFunction=&MF_CharacterParent_2023_PreFX}
  n2 = GetMaterialAttributes(None=@n1.Result) {attributeGetTypes [#1]}
  n3 = TextureSampleParameter2D() {
   group=MadameMoh
   parameterName=FresnelMask
   sortPriority=1
   texture=/BRCosmetics/Characters/Player/Female/Medium/Bodies/F_MED_MadameMoth_Posh/Textures/T_MadameMoth_Ember_TechArt_FX
  }
  n4 = StaticComponentMaskParameter(None=@n3.RGB) {defaultR=true, group=MadameMoh, parameterName=ChannelMask, sortPriority=2}
  n5 = Fresnel(BaseReflectFractionIn=ScalarParameter() {defaultValue=0.1, group=MadameMoh, parameterName=FresnelReflect}) {baseReflectFraction=0.1, exponent=1}
  n6 = Power(Base=@n5, Exp=ScalarParameter() {defaultValue=3, group=MadameMoh, parameterName=FresnelPow}) {constExponent=4}
  n7 = Multiply(A=TextureCoordinate()) {constB=2}
  n8 = Time()
  n9 = Multiply(A=Constant2Vector() {r=1}) {constB=0.1}
  n10 = TextureSample(UVs=Add(A=@n7, B=Multiply(A=@n8, B=@n9))) {samplerSource=SSM_Wrap_WorldGroupSettings, texture=/Game/Global/Textures/Default/Noise/T_FireNoise_01}
  n11 = Multiply(A=@n8, B=Multiply(A=@n9) {constB=-1})
  n12 = TextureSample(UVs=Multiply(A=Add(A=@n7, B=@n11)) {constB=1.1}) {samplerSource=SSM_Wrap_WorldGroupSettings, texture=&T_FireNoise_01}
  n13 = QualitySwitch(Default=Divide(A=Add(A=@n10.R, B=@n12.R)), Low=@n10.R, Medium=@n10.R)
  n14 = Multiply(A=NamedRerouteDeclaration(None=ShadingPathSwitch(Default=@n13, Deferred=@n10.R, Mobile=@n10.R)) {name=FXFresnel, nodeColor=<rgba 0.8984 1 0 1>}) {constB=5}
  n15 = Multiply(A=@n6, B=Add(A=@n6, B=Multiply(A=@n5, B=@n14)))
  n16 = Multiply(
   A=Multiply(
    A=VectorParameter() {group=MadameMoh, parameterName=FresnelColor, defaultValue=<rgba 0.2502 0.0027 0.013 1>}.RGB
    B=ScalarParameter() {defaultValue=15, group=MadameMoh, parameterName=FresnelFlamesIntensity}
   )
   B=@n15
  )
  n17 = Multiply(A=@n4, B=@n15)
  MaterialFunctionCall(
   Material=SetMaterialAttributes!(
    MaterialAttributes=@n2.MaterialAttributes
    Emissive Color=Multiply(A=@n4, B=@n16)
    Base Color=LinearInterpolate(A=@n2.Base Color, Alpha=@n17) {constB=0}
   ) {attributeSetTypes [#2, #1]}
   M=@n1.M
   Pre Skinned Local Position=@n1.M
   Distance Blend=@n1.M
   Fresnel=@n1.M
  ) {materialFunction=&MF_CharacterParent_2023_PostFX}
 ]
}
