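# Physics asset of cape colliders for the Madame Moth Posh female medium player body.
# Capsules cover the spine, clavicles, head and feet, while arm, thigh, calf and neck bodies have no shapes.
# Five four-bone dynamic coat chains also carry capsules: left and right back and side, plus right front only.
# Constraints link neighboring bones with free swing and twist, leaving coat chains and legs unconstrained to the spine.
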PhysicsAsset /BRCosmetics/Characters/Player/Female/Medium/Bodies/F_MED_MadameMoth_Posh/Meshes/F_MED_MadameMoth_Posh_CapeColliders
bodies {
 calf_l []
 calf_r []
 clavicle_l [{shapeType=Capsule, length=6.5623, radius=7.5194, center=<3.9566 -1.0862 -2.1127>, rotation=<rot -77.5566 177.3915 7.7314>}]
 clavicle_r [{shapeType=Capsule, length=6.5623, radius=7.5194, center=<-3.9566 1.0862 2.1127>, rotation=<rot -77.5566 177.3915 -172.2686>}]
 dyn_L_backCoat_1 [{shapeType=Capsule, length=12.4538, radius=8, shapeName=dyn_L_backCoat_1_capsule, center=<3.6425 2.7558 0.3174>, rotation=<rot 90 0 0>}]
 dyn_L_backCoat_2 [{shapeType=Capsule, length=7.4545, radius=8, shapeName=dyn_L_backCoat_2_capsule, center=<3.6425 2.7558 0.3174>, rotation=<rot 90 0 0>}]
 dyn_L_backCoat_3 [{shapeType=Capsule, length=11.225, radius=8, shapeName=dyn_L_backCoat_3_capsule, center=<3.6425 2.7558 0.3174>, rotation=<rot 90 0 0>}]
 dyn_L_backCoat_4 [{shapeType=Capsule, length=9.2474, radius=9, shapeName=dyn_L_backCoat_4_capsule, center=<3.6425 2.7558 0.3174>, rotation=<rot 90 0 0>}]
 dyn_L_sideCoat_1 [{shapeType=Capsule, length=9.2205, radius=8, shapeName=dyn_L_sideCoat_1_capsule, center=<4 1.4 0>, rotation=<rot 90 0 0>}]
 dyn_L_sideCoat_2 [{shapeType=Capsule, length=10.6416, radius=8, shapeName=dyn_L_sideCoat_2_capsule, center=<4 1.4 0>, rotation=<rot 90 0 0>}]
 dyn_L_sideCoat_3 [{shapeType=Capsule, length=11.2535, radius=8, shapeName=dyn_L_sideCoat_3_capsule, center=<4 1.4 0>, rotation=<rot 90 0 0>}]
 dyn_L_sideCoat_4 [{shapeType=Capsule, length=10.2611, radius=8, shapeName=dyn_L_sideCoat_4_capsule, center=<4 1.4 0>, rotation=<rot 90 0 0>}]
 dyn_R_backCoat_1 [{shapeType=Capsule, length=0.505, radius=8, shapeName=dyn_R_backCoat_1_capsule, center=<3.6425 -2.7558 0.3174>, rotation=<rot 90 0 0>}]
 dyn_R_backCoat_2 [{shapeType=Capsule, length=10.1308, radius=8, shapeName=dyn_R_backCoat_2_capsule, center=<3.6425 -2.7558 0.3174>, rotation=<rot 90 0 0>}]
 dyn_R_backCoat_3 [{shapeType=Capsule, length=13.0304, radius=8, shapeName=dyn_R_backCoat_3_capsule, center=<3.6425 -2.7558 0.3174>, rotation=<rot 90 0 0>}]
 dyn_R_backCoat_4 [{shapeType=Capsule, length=11.951, radius=9, shapeName=dyn_R_backCoat_4_capsule, center=<3.6425 -2.7558 0.3174>, rotation=<rot 90 0 0>}]
 dyn_R_frontCoat_1 [{shapeType=Capsule, length=11.7173, radius=8, shapeName=dyn_R_frontCoat_1_capsule, center=<4 0 0>, rotation=<rot 90 0 0>}]
 dyn_R_frontCoat_2 [{shapeType=Capsule, length=11.3752, radius=8, shapeName=dyn_R_frontCoat_2_capsule, center=<4 0 0>, rotation=<rot 90 0 0>}]
 dyn_R_frontCoat_3 [{shapeType=Capsule, length=8.1652, radius=8, shapeName=dyn_R_frontCoat_3_capsule, center=<4 0 0>, rotation=<rot 90 0 0>}]
 dyn_R_frontCoat_4 [{shapeType=Capsule, length=15.6378, radius=8, shapeName=dyn_R_frontCoat_4_capsule, center=<4 0 0>, rotation=<rot 90 0 0>}]
 dyn_R_sideCoat_1 [{shapeType=Capsule, length=11.3007, radius=8, shapeName=dyn_R_sideCoat_1_capsule, center=<4 -1.4 0>, rotation=<rot 90 0 0>}]
 dyn_R_sideCoat_2 [{shapeType=Capsule, length=12.661, radius=8, shapeName=dyn_R_sideCoat_2_capsule, center=<4 -1.4 0>, rotation=<rot 90 0 0>}]
 dyn_R_sideCoat_3 [{shapeType=Capsule, length=13.6712, radius=8, shapeName=dyn_R_sideCoat_3_capsule, center=<4 -1.4 0>, rotation=<rot 90 0 0>}]
 dyn_R_sideCoat_4 [{shapeType=Capsule, length=12.7578, radius=8, shapeName=dyn_R_sideCoat_4_capsule, center=<4 -1.4 0>, rotation=<rot 90 0 0>}]
 foot_l [{shapeType=Capsule, length=11.0765, radius=8.0344, center=<-4.1811 6.0027 -0.338>, rotation=<rot 0.0388 1.4701 -89.3982>}]
 foot_r [{shapeType=Capsule, length=11.0765, radius=8.0344, center=<4.1811 -6.0027 0.338>, rotation=<rot 0.0388 1.4701 90.6018>}]
 head [{shapeType=Capsule, length=0, radius=18, center=<9.0697 0.0746 0>, rotation=<rot 0 -90 0>}]
 lowerarm_l []
 lowerarm_r []
 neck_02 []
 spine_01 [{shapeType=Capsule, length=4.3888, radius=14.3638, center=<-8.9021 -0.8973 0>}]
 spine_02 [{shapeType=Capsule, length=5.2297, radius=9.7317, center=<-2.6885 0.3696 0>}]
 spine_03 [{shapeType=Capsule, length=6.1863, radius=8.4854, center=<0.0574 0.4416 0>}]
 spine_04 [{shapeType=Capsule, length=9.0076, radius=10.3916, center=<2.9889 0.7397 0>}]
 spine_05 [
  {shapeType=Capsule, length=9.678, radius=12.3571, center=<2.6848 1.91 0>}
  {shapeType=Capsule, length=15, radius=13.6807, shapeName=spine_05_capsule, center=<9.6572 2.0639 0>}
 ]
 thigh_l []
 thigh_r []
 upperarm_l []
 upperarm_r []
}
constraints [
 {
  bone1Name=spine_02
  bone2Name=spine_01
  swing1LimitDegrees=45
  swing1Motion=Free
  swing2LimitDegrees=45
  swing2Motion=Free
  twistLimitDegrees=45
  twistMotion=Free
 }
 {
  bone1Name=spine_03
  bone2Name=spine_02
  swing1LimitDegrees=45
  swing1Motion=Free
  swing2LimitDegrees=45
  swing2Motion=Free
  twistLimitDegrees=45
  twistMotion=Free
 }
 {
  bone1Name=spine_04
  bone2Name=spine_03
  swing1LimitDegrees=45
  swing1Motion=Free
  swing2LimitDegrees=45
  swing2Motion=Free
  twistLimitDegrees=45
  twistMotion=Free
 }
 {
  bone1Name=spine_05
  bone2Name=spine_04
  swing1LimitDegrees=45
  swing1Motion=Free
  swing2LimitDegrees=45
  swing2Motion=Free
  twistLimitDegrees=45
  twistMotion=Free
 }
 {
  bone1Name=clavicle_l
  bone2Name=spine_05
  swing1LimitDegrees=45
  swing1Motion=Free
  swing2LimitDegrees=45
  swing2Motion=Free
  twistLimitDegrees=45
  twistMotion=Free
 }
 {
  bone1Name=upperarm_l
  bone2Name=clavicle_l
  swing1LimitDegrees=45
  swing1Motion=Free
  swing2LimitDegrees=45
  swing2Motion=Free
  twistLimitDegrees=45
  twistMotion=Free
 }
 {
  bone1Name=lowerarm_l
  bone2Name=upperarm_l
  swing1LimitDegrees=45
  swing1Motion=Free
  swing2LimitDegrees=45
  swing2Motion=Free
  twistLimitDegrees=45
  twistMotion=Free
 }
 {
  bone1Name=neck_02
  bone2Name=spine_05
  swing1LimitDegrees=45
  swing1Motion=Free
  swing2LimitDegrees=45
  swing2Motion=Free
  twistLimitDegrees=45
  twistMotion=Free
 }
 {
  bone1Name=head
  bone2Name=neck_02
  swing1LimitDegrees=45
  swing1Motion=Free
  swing2LimitDegrees=45
  swing2Motion=Free
  twistLimitDegrees=45
  twistMotion=Free
 }
 {
  bone1Name=calf_l
  bone2Name=thigh_l
  swing1LimitDegrees=45
  swing1Motion=Free
  swing2LimitDegrees=45
  swing2Motion=Free
  twistLimitDegrees=45
  twistMotion=Free
 }
 {
  bone1Name=foot_l
  bone2Name=calf_l
  swing1LimitDegrees=45
  swing1Motion=Free
  swing2LimitDegrees=45
  swing2Motion=Free
  twistLimitDegrees=45
  twistMotion=Free
 }
 {
  bone1Name=lowerarm_r
  bone2Name=upperarm_r
  swing1LimitDegrees=45
  swing1Motion=Free
  swing2LimitDegrees=45
  swing2Motion=Free
  twistLimitDegrees=45
  twistMotion=Free
 }
 {
  bone1Name=upperarm_r
  bone2Name=clavicle_r
  swing1LimitDegrees=45
  swing1Motion=Free
  swing2LimitDegrees=45
  swing2Motion=Free
  twistLimitDegrees=45
  twistMotion=Free
 }
 {
  bone1Name=clavicle_r
  bone2Name=spine_05
  swing1LimitDegrees=45
  swing1Motion=Free
  swing2LimitDegrees=45
  swing2Motion=Free
  twistLimitDegrees=45
  twistMotion=Free
 }
 {
  bone1Name=foot_r
  bone2Name=calf_r
  swing1LimitDegrees=45
  swing1Motion=Free
  swing2LimitDegrees=45
  swing2Motion=Free
  twistLimitDegrees=45
  twistMotion=Free
 }
 {
  bone1Name=calf_r
  bone2Name=thigh_r
  swing1LimitDegrees=45
  swing1Motion=Free
  swing2LimitDegrees=45
  swing2Motion=Free
  twistLimitDegrees=45
  twistMotion=Free
 }
 {
  bone1Name=dyn_L_backCoat_2
  bone2Name=dyn_L_backCoat_1
  swing1LimitDegrees=45
  swing1Motion=Free
  swing2LimitDegrees=45
  swing2Motion=Free
  twistLimitDegrees=45
  twistMotion=Free
 }
 {
  bone1Name=dyn_L_backCoat_3
  bone2Name=dyn_L_backCoat_2
  swing1LimitDegrees=45
  swing1Motion=Free
  swing2LimitDegrees=45
  swing2Motion=Free
  twistLimitDegrees=45
  twistMotion=Free
 }
 {
  bone1Name=dyn_L_backCoat_4
  bone2Name=dyn_L_backCoat_3
  swing1LimitDegrees=45
  swing1Motion=Free
  swing2LimitDegrees=45
  swing2Motion=Free
  twistLimitDegrees=45
  twistMotion=Free
 }
 {
  bone1Name=dyn_L_sideCoat_2
  bone2Name=dyn_L_sideCoat_1
  swing1LimitDegrees=45
  swing1Motion=Free
  swing2LimitDegrees=45
  swing2Motion=Free
  twistLimitDegrees=45
  twistMotion=Free
 }
 {
  bone1Name=dyn_L_sideCoat_3
  bone2Name=dyn_L_sideCoat_2
  swing1LimitDegrees=45
  swing1Motion=Free
  swing2LimitDegrees=45
  swing2Motion=Free
  twistLimitDegrees=45
  twistMotion=Free
 }
 {
  bone1Name=dyn_L_sideCoat_4
  bone2Name=dyn_L_sideCoat_3
  swing1LimitDegrees=45
  swing1Motion=Free
  swing2LimitDegrees=45
  swing2Motion=Free
  twistLimitDegrees=45
  twistMotion=Free
 }
 {
  bone1Name=dyn_R_backCoat_2
  bone2Name=dyn_R_backCoat_1
  swing1LimitDegrees=45
  swing1Motion=Free
  swing2LimitDegrees=45
  swing2Motion=Free
  twistLimitDegrees=45
  twistMotion=Free
 }
 {
  bone1Name=dyn_R_backCoat_3
  bone2Name=dyn_R_backCoat_2
  swing1LimitDegrees=45
  swing1Motion=Free
  swing2LimitDegrees=45
  swing2Motion=Free
  twistLimitDegrees=45
  twistMotion=Free
 }
 {
  bone1Name=dyn_R_backCoat_4
  bone2Name=dyn_R_backCoat_3
  swing1LimitDegrees=45
  swing1Motion=Free
  swing2LimitDegrees=45
  swing2Motion=Free
  twistLimitDegrees=45
  twistMotion=Free
 }
 {
  bone1Name=dyn_R_frontCoat_2
  bone2Name=dyn_R_frontCoat_1
  swing1LimitDegrees=45
  swing1Motion=Free
  swing2LimitDegrees=45
  swing2Motion=Free
  twistLimitDegrees=45
  twistMotion=Free
 }
 {
  bone1Name=dyn_R_frontCoat_3
  bone2Name=dyn_R_frontCoat_2
  swing1LimitDegrees=45
  swing1Motion=Free
  swing2LimitDegrees=45
  swing2Motion=Free
  twistLimitDegrees=45
  twistMotion=Free
 }
 {
  bone1Name=dyn_R_frontCoat_4
  bone2Name=dyn_R_frontCoat_3
  swing1LimitDegrees=45
  swing1Motion=Free
  swing2LimitDegrees=45
  swing2Motion=Free
  twistLimitDegrees=45
  twistMotion=Free
 }
 {
  bone1Name=dyn_R_sideCoat_2
  bone2Name=dyn_R_sideCoat_1
  swing1LimitDegrees=45
  swing1Motion=Free
  swing2LimitDegrees=45
  swing2Motion=Free
  twistLimitDegrees=45
  twistMotion=Free
 }
 {
  bone1Name=dyn_R_sideCoat_3
  bone2Name=dyn_R_sideCoat_2
  swing1LimitDegrees=45
  swing1Motion=Free
  swing2LimitDegrees=45
  swing2Motion=Free
  twistLimitDegrees=45
  twistMotion=Free
 }
 {
  bone1Name=dyn_R_sideCoat_4
  bone2Name=dyn_R_sideCoat_3
  swing1LimitDegrees=45
  swing1Motion=Free
  swing2LimitDegrees=45
  swing2Motion=Free
  twistLimitDegrees=45
  twistMotion=Free
 }
]
